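# Niagara effect for the Street Demon Orange backpack cosmetic that produces a pulsing warm light with soft sprites.
# One emitter spawns short-lived particles through a light renderer, with FN_Light_Attributes driving a red-to-orange color curve and a rise-and-fall intensity curve at a fixed local offset.
# A second emitter spawns small sprites at an offset from the system location, fades their alpha in and out by curve, and nudges them toward the camera.
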NiagaraSystem /BRCosmetics/Accessories/FORT_Backpacks/Backpack_M_MED_Street_Demon_Orange/FX/FN_Backpack_StreetDemonOrange_Light
dependencies {
 usedDataInterfaces [&NiagaraDataInterfaceColorCurve, &NiagaraDataInterfaceCurve]
 usedDynamicInputs []
 usedModules [&SystemState]
}
emitters {
 Empty {
  emitterUpdateScript [
   &EmitterState(Life Cycle Mode=::NewEnumerator1, Loop Duration=1, UseLoopDelay=false)
   &SpawnRate(SpawnRate=1.2, Use Spawn Probability=false)
  ]
  particleSpawnScript [&InitializeParticle(
    Color Mode=::NewEnumerator0
    Lifetime Max=1.4
    Lifetime Min=0.86
    Lifetime Mode=::NewEnumerator1
    UsePositionOffset=false
    Write Material Random=true
   )]
  particleUpdateScript [
   &ParticleState()
   &FN_Light_Attributes(
    Color=&ColorFromCurve(ColorCurve=DataInterface() {propertyValues {
       AlphaCurve {keys [<at 0 0 RCIM_Cubic>,<at 0.1685 1>,<at 0.5312 1>,<at 1 0 RCIM_Cubic>]}
       BlueCurve {keys [<at 0 0 RCIM_Cubic>,<at 0.2496 0>,<at 1 0 RCIM_Cubic>]}
       GreenCurve {keys [<at 0 0.1036 RCIM_Cubic>,<at 0.2496 0.153>,<at 1 0.3819 RCIM_Cubic>]}
       RedCurve {keys [<at 0 1 RCIM_Cubic>,<at 0.2496 1>,<at 1 1 RCIM_Cubic>]}
      }})
    Create Color Attribute=true
    Create Exponent Attribute=true
    Create Position Attribute=true
    Create Radius Attribute=true
    Exponent=3
    Position=<0 -7 0>
    Radius=16
    Scale Color=&Multiply_Float(
     A=5
     B=&FloatFromCurve(FloatCurve=DataInterface() {propertyValues {Curve {keys [
          {arriveTangent=8.4829, interpMode=RCIM_Cubic, leaveTangent=8.4829, tangentMode=RCTM_User, time=0, value=0}
          <at 0.2 1 RCIM_Cubic>
          {arriveTangent=-3.5141, interpMode=RCIM_Cubic, leaveTangent=-3.5141, tangentMode=RCTM_User, time=1, value=0}
         ]}}})
    )
   )
  ]
  rendererClasses [&NiagaraLightRendererProperties]
  renderers [{rendererClass=&NiagaraLightRendererProperties, rendererIndex=0}]
 }
 Empty001 {
  emitterUpdateScript [
   &EmitterState()
   &SpawnRate(SpawnRate=2, Use Spawn Probability=false)
  ]
  particleSpawnScript [
   &InitializeParticle(
    Color Mode=::NewEnumerator0
    Lifetime Max=1
    Lifetime Mode=::NewEnumerator1
    Sprite Size Mode=::NewEnumerator3
    Uniform Sprite Size=5
    UsePositionOffset=false
    Write Material Random=true
   )
   &SystemLocation(MaskSpawnGroup=false, Offset=<0 -3 0>)
  ]
  particleUpdateScript [
   &ParticleState()
   &ScaleColor(
    Scale Alpha=&Multiply_Float(
     A=0.3
     B=&FloatFromCurve(FloatCurve=DataInterface() {propertyValues {Curve {keys [
          {arriveTangent=4.9348, interpMode=RCIM_Cubic, leaveTangent=4.9348, tangentMode=RCTM_User, time=0, value=0}
          <at 0.3 1.05 RCIM_Cubic>
          {arriveTangent=-2.9659, interpMode=RCIM_Cubic, leaveTangent=-2.9659, tangentMode=RCTM_User, time=1, value=0}
         ]}}})
    )
    Scale RGB=&MakeVectorFromLinearColorRGB(Color=<rgba 1 1 1 1>)
    ScaleA=true
    ScaleRGB=true
   )
   &CameraOffset(Camera Offset Amount=2)
  ]
  rendererClasses [&NiagaraSpriteRendererProperties]
  renderers [{rendererClass=&NiagaraSpriteRendererProperties}]
 }
}
summary {
 systemName=FN_Backpack_StreetDemonOrange_Light
 emitters [
  {emitterName=Empty, rendererClasses [&NiagaraLightRendererProperties]}
  {emitterName=Empty001, rendererClasses [&NiagaraSpriteRendererProperties]}
 ]
 userVariables []
}
